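# Sound cue for the SleepyTime cosmetic pickaxe hitting a player, using its enemy impact sounds.
# A root Random node picks one of six equally weighted wave players, each holding a numbered impact variant.
# It primes on load and routes through the Fort_Impacts sound class, a shared pickaxe impact attenuation asset and audio property sheet.
# The shared melee weapon concurrency set governs simultaneous playback.
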SoundCue /BRCosmetics/Sounds/PickAxes/SleepyTime/PickAxe_SleepyTime_Impact_Player_Athena_Cue
root=Random_1
nodes [
 {
  id=Random_1
  children [WavePlayer_5, WavePlayer_6, WavePlayer_7, WavePlayer_8, WavePlayer_10, WavePlayer_11]
  properties {weights [1,1,1,1,1,1]}
 }
 {id=WavePlayer_10, properties {soundWaveAssetPtr=&/PickAxe_SleepyTime_EnemyImpact_04}}
 {id=WavePlayer_11, properties {soundWaveAssetPtr=&/PickAxe_SleepyTime_EnemyImpact_05}}
 {id=WavePlayer_5, properties {soundWaveAssetPtr=&/PickAxe_SleepyTime_EnemyImpact_06}}
 {id=WavePlayer_6, properties {soundWaveAssetPtr=&/PickAxe_SleepyTime_EnemyImpact_01}}
 {id=WavePlayer_7, properties {soundWaveAssetPtr=&/PickAxe_SleepyTime_EnemyImpact_02}}
 {id=WavePlayer_8, properties {soundWaveAssetPtr=&/PickAxe_SleepyTime_EnemyImpact_03}}
]
properties {
 attenuationSettings=/Game/Sounds/Attenuation/Fort_Pickaxe_Impact_Att
 audioPropertiesSheet=/BRCosmetics/Sounds/Settings/PropertySheets/Pickaxe/APS_Cosmetic_Pickaxe_Impact
 bPrimeOnLoad=true
 maxDistance=4300
 soundClassObject=&Fort_Impacts
 attenuationOverrides {
  bAttenuateWithLPF=true
  distanceAlgorithm=Logarithmic
  falloffDistance=800
  lPFRadiusMax=800
  lPFRadiusMin=400
  attenuationShapeExtents {x=300}
 }
 concurrencyOverrides {bLimitToOwner=true, maxCount=2, resolutionRule=StopOldest}
 concurrencySet [&Player_Melee_Weapon_Conc]
}
